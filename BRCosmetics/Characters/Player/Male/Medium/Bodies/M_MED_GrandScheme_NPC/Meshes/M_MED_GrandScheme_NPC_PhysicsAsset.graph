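# Physics asset for the medium male GrandScheme NPC body, giving its skeleton a simplified collision and joint setup.
# Single capsules cover pelvis, spine, head, upper arms, hands and legs, with no forearm bodies, so hands constrain directly to upper arms.
# All constraints share uniform swing and twist limits.
# The left leg is asymmetric: thigh_twist_01_l replaces a thigh body, and calf_l constrains straight to the pelvis.
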PhysicsAsset /BRCosmetics/Characters/Player/Male/Medium/Bodies/M_MED_GrandScheme_NPC/Meshes/M_MED_GrandScheme_NPC_PhysicsAsset
bodies {
 calf_l [{shapeType=Capsule, length=20.8225, radius=9.2275, center=<-20.4117 -1.1089 0.5158>, rotation=<rot 0 -85.9495 -88.0592>}]
 calf_r [{shapeType=Capsule, length=20.8225, radius=9.2275, center=<20.4117 1.1089 -0.5158>, rotation=<rot 0 -85.9496 -88.0592>}]
 foot_l [{shapeType=Capsule, length=16.1826, radius=7.8064, center=<-3.4909 6.9836 -0.5579>, rotation=<rot 0 8.1399 -89.0837>}]
 foot_r [{shapeType=Capsule, length=16.178, radius=7.8, center=<3.4789 -6.9844 0.5488>, rotation=<rot 0 8.0608 -88.9731>}]
 hand_l [{shapeType=Capsule, length=8.8986, radius=6.5921, center=<7.7652 -2.2244 0.0482>, rotation=<rot 0 83.3377 -65.1569>}]
 hand_r [{shapeType=Capsule, length=8.8986, radius=6.5921, center=<-7.7652 2.2244 -0.0482>, rotation=<rot 0 83.3377 -65.1569>}]
 head [{shapeType=Capsule, length=15.1341, radius=11.1741, center=<2.9156 2.0785 0>, rotation=<rot 0 45.3013 -89.9993>}]
 pelvis [{shapeType=Capsule, length=33.9456, radius=17.7968, center=<-12.0264 1.9772 -8.77>, rotation=<rot 0 67.617 -66.2186>}]
 spine_01 [{shapeType=Capsule, length=17.3966, radius=12.7269, center=<7.8436 0.9438 -0.0029>, rotation=<rot 0.3023 -90 -0.152>}]
 spine_03 [{shapeType=Capsule, length=15.9048, radius=14.6307, center=<3.3059 -0.1595 -0.0458>, rotation=<rot -1.8055 -90 0.1867>}]
 spine_04 [{shapeType=Capsule, length=16.752, radius=15.7368, center=<2.8699 -1.0014 -0.0138>, rotation=<rot 0.453 -90 0.2739>}]
 spine_05 [{shapeType=Capsule, length=20.4545, radius=14.6945, center=<7.0461 -1.795 0>, rotation=<rot 0 -90 0>}]
 thigh_r [{shapeType=Capsule, length=23.2492, radius=10.685, center=<21.3836 -1.1704 -0.5478>, rotation=<rot 0 -96.4353 -85.7696>}]
 thigh_twist_01_l [{shapeType=Capsule, length=15.6957, radius=12.6194, center=<7.6029 1.5146 0.3903>, rotation=<rot 0 -48.8139 -79.8313>}]
 upperarm_l [{shapeType=Capsule, length=25.6905, radius=9.3285, center=<23.2594 2.5033 -0.9444>, rotation=<rot 0 -78.1032 -87.0563>}]
 upperarm_r [{shapeType=Capsule, length=25.6906, radius=9.3285, center=<-23.2594 -2.5033 0.9444>, rotation=<rot 0 -78.1032 -87.0563>}]
}
constraints [
 {bone1Name=spine_01, bone2Name=pelvis, swing1LimitDegrees=45, swing2LimitDegrees=45, twistLimitDegrees=45}
 {bone1Name=spine_03, bone2Name=spine_01, swing1LimitDegrees=45, swing2LimitDegrees=45, twistLimitDegrees=45}
 {bone1Name=spine_04, bone2Name=spine_03, swing1LimitDegrees=45, swing2LimitDegrees=45, twistLimitDegrees=45}
 {bone1Name=spine_05, bone2Name=spine_04, swing1LimitDegrees=45, swing2LimitDegrees=45, twistLimitDegrees=45}
 {bone1Name=head, bone2Name=spine_05, swing1LimitDegrees=45, swing2LimitDegrees=45, twistLimitDegrees=45}
 {bone1Name=upperarm_l, bone2Name=spine_05, swing1LimitDegrees=45, swing2LimitDegrees=45, twistLimitDegrees=45}
 {bone1Name=hand_l, bone2Name=upperarm_l, swing1LimitDegrees=45, swing2LimitDegrees=45, twistLimitDegrees=45}
 {bone1Name=upperarm_r, bone2Name=spine_05, swing1LimitDegrees=45, swing2LimitDegrees=45, twistLimitDegrees=45}
 {bone1Name=hand_r, bone2Name=upperarm_r, swing1LimitDegrees=45, swing2LimitDegrees=45, twistLimitDegrees=45}
 {bone1Name=calf_l, bone2Name=pelvis, swing1LimitDegrees=45, swing2LimitDegrees=45, twistLimitDegrees=45}
 {bone1Name=foot_l, bone2Name=calf_l, swing1LimitDegrees=45, swing2LimitDegrees=45, twistLimitDegrees=45}
 {bone1Name=thigh_twist_01_l, bone2Name=pelvis, swing1LimitDegrees=45, swing2LimitDegrees=45, twistLimitDegrees=45}
 {bone1Name=thigh_r, bone2Name=pelvis, swing1LimitDegrees=45, swing2LimitDegrees=45, twistLimitDegrees=45}
 {bone1Name=calf_r, bone2Name=thigh_r, swing1LimitDegrees=45, swing2LimitDegrees=45, twistLimitDegrees=45}
 {bone1Name=foot_r, bone2Name=calf_r, swing1LimitDegrees=45, swing2LimitDegrees=45, twistLimitDegrees=45}
]
